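# Sound cue for the LyricalMale cosmetic pickaxe's enemy impact sound, routed to the Fort_Impacts sound class.
# A Random node picks among ten equally weighted impact waves, then a Modulator adds slight pitch and volume variation.
# A Delay node at the root holds playback back by a short fixed time.
# It uses shared pickaxe impact attenuation and property sheet, with a concurrency override capping simultaneous plays.
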SoundCue /BRCosmetics/Sounds/PickAxes/LyricalMale/PickaxeImpactEnemy_LyricalMale_Cue
root=Delay_0
nodes [
 {id=Delay_0, children [Modulator_0], properties {delayMax=0.1, delayMin=0.1}}
 {id=Modulator_0, children [Random_0], properties {pitchMax=1.02, pitchMin=0.98, volumeMax=1.01, volumeMin=0.99}}
 {
  id=Random_0
  children [WavePlayer_0,WavePlayer_1,WavePlayer_2,WavePlayer_3,WavePlayer_4,WavePlayer_5,WavePlayer_6,WavePlayer_7,WavePlayer_8,WavePlayer_9]
  properties {weights [1,1,1,1,1,1,1,1,1,1]}
 }
 {id=WavePlayer_0, properties {soundWaveAssetPtr=&/Pickaxe_LyricalMale_Impact_01}}
 {id=WavePlayer_1, properties {soundWaveAssetPtr=&/Pickaxe_LyricalMale_Impact_02}}
 {id=WavePlayer_2, properties {soundWaveAssetPtr=&/Pickaxe_LyricalMale_Impact_03}}
 {id=WavePlayer_3, properties {soundWaveAssetPtr=&/Pickaxe_LyricalMale_Impact_04}}
 {id=WavePlayer_4, properties {soundWaveAssetPtr=&/Pickaxe_LyricalMale_Impact_05}}
 {id=WavePlayer_5, properties {soundWaveAssetPtr=&/Pickaxe_LyricalMale_Impact_06}}
 {id=WavePlayer_6, properties {soundWaveAssetPtr=&/Pickaxe_LyricalMale_Impact_07}}
 {id=WavePlayer_7, properties {soundWaveAssetPtr=&/Pickaxe_LyricalMale_Impact_08}}
 {id=WavePlayer_8, properties {soundWaveAssetPtr=&/Pickaxe_LyricalMale_Impact_09}}
 {id=WavePlayer_9, properties {soundWaveAssetPtr=&/Pickaxe_LyricalMale_Impact_10}}
]
properties {
 attenuationSettings=/Game/Sounds/Attenuation/Fort_Pickaxe_Impact_Att
 audioPropertiesSheet=/BRCosmetics/Sounds/Settings/PropertySheets/Pickaxe/APS_Cosmetic_Pickaxe_Impact
 bOverrideConcurrency=true
 maxDistance=4300
 soundClassObject=&Fort_Impacts
 concurrencyOverrides {maxCount=2}
 concurrencySet [null]
}
